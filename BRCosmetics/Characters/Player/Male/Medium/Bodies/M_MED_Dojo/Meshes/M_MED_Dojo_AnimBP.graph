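# Animation Blueprint for the M_MED_Dojo male player body cosmetic, copying its pose from the attached parent mesh and correcting kneepad and elbowpad bones.
# Each update turns the crouch flag into a kneepad alpha and maps each lower arm's swing angle from reference pose to an elbowpad alpha.
# These alphas drive chained additive bone-space translation and rotation offsets on the four pad bones, with interpolation smoothing them, much faster for knees than elbows.
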AnimBlueprint /BRCosmetics/Characters/Player/Male/Medium/Bodies/M_MED_Dojo/Meshes/M_MED_Dojo_AnimBP
target_skeleton=&/M_MED_Dojo_Skeleton
graphs [
 AnimGraph {nodes [
   n1 = Get Crouch_Alpha() -> (Crouch_Alpha:double)
   n2 = "Transform (Modify) Bone - Bone: const_kneepad_l"(
    ComponentPose=Local To Component(LocalPose=Copy Pose From Mesh() {bUseAttachedParent=true})
    Alpha=@n1
    Translation="2,0.000,-4"
    Rotation="40,-0.000,-0.000"
    Scale="1,1,1"
   ) {
    rotationMode=BMM_Additive
    rotationSpace=BCS_BoneSpace
    translationMode=BMM_Additive
    translationSpace=BCS_BoneSpace
    alphaBoolBlend {blendOption=Linear}
    alphaScaleBiasClamp {bInterpResult=true, interpSpeedDecreasing=50, interpSpeedIncreasing=50}
    boneToModify {boneName=const_kneepad_l}
    rotation {pitch=40}
    translation {x=2, z=-4}
   }
   n3 = "Transform (Modify) Bone - Bone: const_kneepad_r"(ComponentPose=@n2, Alpha=@n1, Translation="2,0.000,-4", Rotation="40,-0.000,-0.000", Scale="1,1,1") {
    rotationMode=BMM_Additive
    rotationSpace=BCS_BoneSpace
    translationMode=BMM_Additive
    translationSpace=BCS_BoneSpace
    alphaBoolBlend {blendOption=Linear}
    alphaScaleBiasClamp {bInterpResult=true, interpSpeedDecreasing=50, interpSpeedIncreasing=50}
    boneToModify {boneName=const_kneepad_r}
    rotation {pitch=40}
    translation {x=2, z=-4}
   }
   n4 = "Transform (Modify) Bone - Bone: const_elbowpad_l"(ComponentPose=@n3, Alpha=Get ArmSwingDelta_l() -> (ArmSwingDelta_l:double), Translation="0,0,3", Rotation="-30,0.000,0.000", Scale="1,1,1") {
    rotationMode=BMM_Additive
    rotationSpace=BCS_BoneSpace
    translationMode=BMM_Additive
    translationSpace=BCS_BoneSpace
    alphaBoolBlend {blendOption=Linear}
    alphaScaleBiasClamp {bInterpResult=true, interpSpeedDecreasing=4, interpSpeedIncreasing=6}
    boneToModify {boneName=const_elbowpad_l}
    rotation {pitch=-30}
    translation {z=3}
   }
   n5 = "Transform (Modify) Bone - Bone: const_elbowpad_r"(ComponentPose=@n4, Alpha=Get ArmSwingDelta_r() -> (ArmSwingDelta_r:double), Translation="0,0,3", Rotation="-30,0.000,0.000", Scale="1,1,1") {
    rotationMode=BMM_Additive
    rotationSpace=BCS_BoneSpace
    translationMode=BMM_Additive
    translationSpace=BCS_BoneSpace
    alphaBoolBlend {blendOption=Linear}
    alphaScaleBiasClamp {bInterpResult=true, interpSpeedDecreasing=4, interpSpeedIncreasing=6}
    boneToModify {boneName=const_elbowpad_r}
    rotation {pitch=-30}
    translation {z=3}
   }
   Output Pose(Result=Component To Local(ComponentPose=@n5)) {layerGroup=None}
  ]}
 EventGraph {nodes [
   Event BlueprintUpdateAnimation()
   > n1 = Sequence()
   > Set Crouch_Alpha(Crouch_Alpha:double=SelectFloat(A=1, B=0, bPickA=Get bIsCrouching())) -> (Output_Get:double)
   n2 = GetOwnerSkeletalMeshComponent()
   Set ArmSwingDelta_r(
    execute=@n1.then_1
    ArmSwingDelta_r:double=MapRangeClamped(Value=GetTwistAndSwingAngleOfDeltaRotationFromRefPose(self=@n2, BoneName=lowerarm_r).OutSwingAngle, InRangeA=23, InRangeB=120)
   ) -> (Output_Get:double)
   > Set ArmSwingDelta_l(ArmSwingDelta_l:double=MapRangeClamped(Value=GetTwistAndSwingAngleOfDeltaRotationFromRefPose(self=@n2, BoneName=lowerarm_l).OutSwingAngle, InRangeA=31, InRangeB=103)) -> (Output_Get:double)
  ]}
]
variables [
 ArmSwingDelta_l {type=number}
 ArmSwingDelta_r {type=number}
 Crouch_Alpha {type=number}
]
